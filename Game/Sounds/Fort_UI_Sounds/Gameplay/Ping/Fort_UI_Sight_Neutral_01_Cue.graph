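# Sound cue for the neutral "sight" ping UI sound, part of the in-game gameplay ping audio set.
# A single wave player plays the Fort_UI_Sight_Neutral_01 wave, routed through the non-critical in-game UI sound class and the in-game ping attenuation.
# A concurrency override allows only one instance at a time, with a new play stopping the oldest.
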SoundCue /Game/Sounds/Fort_UI_Sounds/Gameplay/Ping/Fort_UI_Sight_Neutral_01_Cue
root=WavePlayer_0
nodes [{id=WavePlayer_0, properties {soundWaveAssetPtr=&/Fort_UI_Sight_Neutral_01}}]
properties {
 attenuationSettings=/Game/Sounds/Attenuation/Fort_UI_Ping_Ingame_Att
 bOverrideConcurrency=true
 maxDistance=2097152
 soundClassObject=&Fort_UI_InGame_NonCritical
 volumeMultiplier=1
 assetUserData [&self:AudioAssetUserData_0]
 concurrencyOverrides {maxCount=1, resolutionRule=StopOldest}
}
