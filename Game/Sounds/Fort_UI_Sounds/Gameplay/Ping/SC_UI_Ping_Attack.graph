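# Sound cue for the attack ping UI sound, routed to the non-critical in-game UI sound class with in-game ping attenuation.
# A mixer layers a quiet tutorial target-eliminated counter wave under a full-volume enemy-sighted wave.
# Each layer passes through a modulator that applies a fixed upward pitch shift without randomization.
# A concurrency override allows one instance at a time, stopping the oldest when a new one plays.
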SoundCue /Game/Sounds/Fort_UI_Sounds/Gameplay/Ping/SC_UI_Ping_Attack
root=Mixer_0
nodes [
 {id=Mixer_0, children [Modulator_2,Modulator_1], properties {inputVolume [0.2,1]}}
 {id=Modulator_1, children [WavePlayer_1], properties {pitchMax=1.1, pitchMin=1.1, volumeMax=1, volumeMin=1}}
 {id=Modulator_2, children [WavePlayer_0], properties {pitchMax=1.2, pitchMin=1.2, volumeMax=1, volumeMin=1}}
 {id=WavePlayer_0, properties {soundWaveAssetPtr=/Game/Sounds/Fort_UI_Sounds/Tutorial/UI_Tutorial_TargetElminated_Counter_01}}
 {id=WavePlayer_1, properties {soundWaveAssetPtr=&/Fort_UI_Sight_Enemy_03_Alt}}
]
properties {
 attenuationSettings=/Game/Sounds/Attenuation/Fort_UI_Ping_Ingame_Att
 bOverrideConcurrency=true
 maxDistance=2097152
 soundClassObject=&Fort_UI_InGame_NonCritical
 volumeMultiplier=0.8
 concurrencyOverrides {maxCount=1, resolutionRule=StopOldest}
}
